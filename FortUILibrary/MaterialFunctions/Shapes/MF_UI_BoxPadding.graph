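# UI material function that draws a rounded box with per-side pixel padding that stays constant at any widget size, outputting Fill, Outline and both raw SDFs.
# Padding is converted from pixels to normalized units, inverted and scaled to size MF_SDF_Box_V4.
# MF_UI_ShapeOutline derives the outline SDF using pixel thickness and alignment.
# Fill and Outline use anti-aliasing, switching to a SmoothStep glow when either falloff value is non-zero.
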MaterialFunction /FortUILibrary/MaterialFunctions/Shapes/MF_UI_BoxPadding
graph {
 properties {
  bExposeToLibrary=true
  description="Draws a UI box with fill, outline, glow, and rounded corner options, plus a new padding mode set in pixels for top, bottom, left, and right, just like in the UMG Designer. The padding size you set in pixels, either uniform or per-side (e.g., float or Vector4), remains constant regardless of the widget’s size."
  previewMaterialDomain=MD_UI
  libraryCategoriesText [UI Library]
 }
 nodes [
  Constant4Vector() {constant=<rgba 0.5 0.2 0.1 0>}
  Constant4Vector() {constant=<rgba 10 8 6 4>}
  Constant() {desc=testing value in pixels, r=2}
  Constant()
  Constant() {desc="The alignment of the outline (0.0 = center, -1.0 = inner, 1.0 = outer).", r=-1}
  Constant() {r=10}
  Constant()
  Constant()
  FunctionOutput(None=NamedRerouteUsage()) {
   description="Outputs the raw signed distance field (SDF) of the shape's outline. No anti-aliasing applied. Used internally to generate the Outline output."
   outputName=SDF Outline
   sortPriority=3
  }
  n1 = NamedRerouteUsage()
  n2 = SmoothStep(Min=NamedRerouteUsage(), Max=NamedRerouteUsage(), Value=NamedRerouteUsage())
  n3 = NamedRerouteUsage()
  FunctionOutput(None=LinearInterpolate(
    A=MaterialFunctionCall(DistanceField=@n1) {materialFunction=&MF_UI_AntiAliasing}.Inner
    B=OneMinus(None=@n2)
    Alpha=@n3
   )) {
   description=Outputs the final shape with anti-aliasing or optional glow applied. This is the main render output for the filled area.
   outputName=Fill
  }
  FunctionOutput(None=NamedRerouteUsage()) {
   description="Outputs the raw signed distance field (SDF) of the shape. No anti-aliasing applied. Used internally to generate the Fill output."
   outputName=SDF
   sortPriority=2
  }
  n4 = NamedRerouteUsage()
  n5 = SmoothStep(Min=NamedRerouteUsage(), Max=NamedRerouteUsage(), Value=NamedRerouteUsage())
  FunctionOutput(None=LinearInterpolate(
    A=MaterialFunctionCall(DistanceField=@n4) {materialFunction=&MF_UI_AntiAliasing}.Inner
    B=OneMinus(None=@n5)
    Alpha=@n3
   )) {
   description="Outputs the shape’s outline with anti-aliasing or optional glow applied. Controlled by the Outline Thickness and Outline Alignment parameters."
   outputName=Outline
   sortPriority=1
  }
  n6 = Add(A=Abs(None=NamedRerouteUsage()), B=Abs(None=NamedRerouteUsage()))
  NamedRerouteDeclaration(None=Saturate(None=Ceil(None=@n6))) {
   name=Is Glow Active
   desc="If Glow Min or Glow Max are not equal to 0, then result 1.\r\nOtherwise result 0."
   nodeColor=<rgba 0 0.117 0.102 1>
  }
  n7 = FunctionInput() {
   bUsePreviewValueAsDefault=true
   description="Applied to: Padding / Outline. Apply to Corners outside of the function if needed. Scales the shape relative to the widget container. Accepts ratio values only. A value of 1 applies no scaling; lower or higher values shrink or enlarge the shape. Padding is applied before this scale. Useful when working with pixel-based sizing but needing visual scaling (e.g., for UI animations)."
   inputName="Scale (ratio 0-1)"
   inputType=FunctionInput_Scalar
   sortPriority=1
   previewValue=<1 0 0 1>
  }
  NamedRerouteDeclaration(None=@n7) {name=Scale, nodeColor=<rgba 0 0 0 1>}
  n8 = FunctionInput(Preview=TextureCoordinate()) {
   bUsePreviewValueAsDefault=true
   description="Controls the UV mapping used to sample and shape the geometry. Use custom UVs to stretch, deform, or reposition the shape within the widget space."
   inputName=UVs
   inputType=FunctionInput_Vector2
  }
  n9 = FunctionInput() {
   bUsePreviewValueAsDefault=true
   description="Controls the padding around the shape. Use a float (scalar) for uniform padding, or a float4 to define individual padding per side. The values are interpreted as absolute pixel values. When padding is set to 0, the shape matches the size of the widget image receiving the material, effectively filling the widget."
   inputName="Padding (px)"
   inputType=FunctionInput_Vector4
   sortPriority=2
   previewValue=<0 0 0 0>
  }
  n10 = NamedRerouteDeclaration(None=@n9) {name=Padding, nodeColor=<rgba 0 0 0 1>}
  n11 = OneMinus(None=MaterialFunctionCall!(
    Value in Pixels=Multiply(A=@n10) {constB=2, desc="to match the -1..,1 domain.\r\nlike that 1px padding = 1px on the viewport."}
    Pixel Reference=NamedRerouteUsage()
   ) {materialFunction=/FortUILibrary/MaterialFunctions/Basics/MF_UI_PixelsToNormalizedV4}."Per‑Axis Normalized (V4)") {desc=Transform Size to Padding}
  n12 = NamedRerouteUsage()
  n13 = FunctionInput() {
   bUsePreviewValueAsDefault=true
   description="Use a float (scalar) for uniform roundness, or a float4 to set a different radius for each corner. The value is interpreted as a ratio (0 to 1) relative to the widget container. Note: Values greater than 0.5 can cause corner misalignment at smaller widget sizes."
   inputName="Corner Roundness (ratio 0-1)"
   inputType=FunctionInput_Vector4
   sortPriority=3
   previewValue=<0 0 0 0>
  }
  n14 = FunctionInput() {
   bUsePreviewValueAsDefault=true
   description="0 = False, 1 = True (default). Preserves the proportions of rounded corners to avoid distortion when the widget is non-square. Has no visible effect if corner roundness is 0."
   inputName=Preserve Roundness
   inputType=FunctionInput_Scalar
   sortPriority=4
   previewValue=<1 0 0 1>
  }
  NamedRerouteDeclaration(None=MaterialFunctionCall!(
    UVs=NamedRerouteDeclaration(None=@n8) {name=UVs, nodeColor=<rgba 0 0 0 1>}
    Size=Multiply(A=@n11, B=@n12)
    Corner Roundness=@n13
    Preserve Roundness=@n14
    Pixel Size=NamedRerouteUsage()
   ) {materialFunction=/FortUILibrary/MaterialFunctions/Shapes/SDFs/MF_SDF_Box_V4}.SDF) {name=SDF Shape, nodeColor=<rgba 0 0 0 1>}
  n15 = FunctionInput() {
   bUsePreviewValueAsDefault=true
   description="Controls the alignment of the outline (0.0 = centered, -1.0 = inner (default), 1.0 = outer)."
   inputName=Outline Alignment
   inputType=FunctionInput_Scalar
   sortPriority=6
   previewValue=<-1 0 0 1>
  }
  NamedRerouteDeclaration(None=@n15) {name=Outline Alignment, nodeColor=<rgba 0 0 0 1>}
  NamedRerouteDeclaration(None=MaterialFunctionCall!(DistanceField=NamedRerouteUsage(), Threshold, Thickness=NamedRerouteUsage(), Alignment=NamedRerouteUsage(), Smoothing) {materialFunction=/FortUILibrary/MaterialFunctions/Shapes/Modifiers/MF_UI_ShapeOutline}.SDFOutline) {name=SDF Outline, nodeColor=<rgba 0 0 0 1>}
  n16 = FunctionInput() {
   bUsePreviewValueAsDefault=true
   description="Sets the threshold at which glow falloff begins — this is the lower (minimum) value used in a Smoothstep. Applied to both the SDF shape and SDF outline to create soft edges, glow, or shadow effects. Must be less than Glow Falloff End, or results will invert. Disables anti-aliasing when both glow values are non-zero."
   inputName=Glow Falloff Start
   inputType=FunctionInput_Scalar
   sortPriority=7
  }
  NamedRerouteDeclaration(None=@n16) {name=Glow Min, nodeColor=<rgba 0 0 0 1>}
  n17 = FunctionInput() {
   bUsePreviewValueAsDefault=true
   description="Controls the thickness of the outline when using an outline output. The value is interpreted as  an absolute pixel value."
   inputName="Outline Thickness (px)"
   inputType=FunctionInput_Scalar
   sortPriority=5
   previewValue=<2 0 0 1>
  }
  n18 = MaterialFunctionCall(Value in Pixels=NamedRerouteDeclaration(None=@n17) {name=Outline Thickness, nodeColor=<rgba 0 0 0 1>}) {materialFunction=&MF_UI_PixelsToNormalizedV4}
  NamedRerouteDeclaration(None=Multiply(A=ComponentMask(None=@n18."Normalized (V4)") {r=true}, B=NamedRerouteUsage())) {name=Normalized Outline Thickness, nodeColor=<rgba 0 0 0 1>}
  n19 = FunctionInput() {
   bUsePreviewValueAsDefault=true
   description="Sets the threshold at which glow falloff ends — this is the upper (maximum) value used in a Smoothstep. Applied to both the SDF shape and SDF outline to create soft edges, glow, or shadow effects. Must be greater than Glow Falloff Start, or results will invert. Disables anti-aliasing when both glow values are non-zero."
   inputName=Glow Falloff End
   inputType=FunctionInput_Scalar
   sortPriority=8
  }
  NamedRerouteDeclaration(None=@n19) {name=Glow Max, nodeColor=<rgba 0 0 0 1>}
  n20 = FunctionInput(Preview=MaterialFunctionCall() {materialFunction=&GetUserInterfaceUV}.Pixel Size) {
   bUsePreviewValueAsDefault=true
   description=Pixel Size Reference
   inputName=Pixel Size
   inputType=FunctionInput_Vector2
   sortPriority=100
   previewValue=<0.5 0.5 0 1>
  }
  NamedRerouteDeclaration(None=@n20) {name=Pixel Size, nodeColor=<rgba 0.0078 1 0 1>}
 ]
}
