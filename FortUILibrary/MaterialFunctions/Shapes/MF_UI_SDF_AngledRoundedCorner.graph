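# UI material function that draws a signed distance field for a single rounded corner with adjustable angle, rotation, offset and radius.
# Rotates and offsets the UV, then counter-rotates two copies by an angle-derived amount to form two edge half-planes.
# A box-style SDF (min/max plus length) rounds their junction by half the Corner Radius.
# Outputs raw fill and abs-based outline SDFs, plus anti-aliased Fill and Outline masks through MF_UI_AntiAliasing.
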MaterialFunction /FortUILibrary/MaterialFunctions/Shapes/MF_UI_SDF_AngledRoundedCorner
graph {
 properties {bExposeToLibrary=true, previewMaterialDomain=MD_UI}
 nodes [
  n1 = FunctionInput() {bUsePreviewValueAsDefault=true, inputName=Angle, inputType=FunctionInput_Scalar, sortPriority=2, previewValue=<90 0 0 1>}
  n2 = LinearInterpolate(Alpha=Divide(A=@n1) {constB=180}) {constA=0.125, constB=-0.125}
  n3 = FunctionInput() {bUsePreviewValueAsDefault=true, inputName=Rotation, inputType=FunctionInput_Scalar, sortPriority=1, previewValue=<0.875 0 0 1>}
  n4 = Rotator(
   Coordinate=FunctionInput(Preview=TextureCoordinate()) {bUsePreviewValueAsDefault=true, inputName=UV, inputType=FunctionInput_Vector2}
   Time=@n3
  ) {speed=6.2832}
  n5 = FunctionInput() {bUsePreviewValueAsDefault=true, inputName=Offset, inputType=FunctionInput_Vector2, sortPriority=1, previewValue=<0.25 0.25 0 1>}
  n6 = FunctionInput() {bUsePreviewValueAsDefault=true, inputName=Corner Radius, inputType=FunctionInput_Scalar, sortPriority=4, previewValue=<0.25 0 0 1>}
  n7 = Multiply(A=@n6) {constB=0.5}
  n8 = Add(A=Subtract(A=Add(A=@n4, B=@n5)), B=@n7)
  n9 = ComponentMask(None=Rotator(Coordinate=@n8, Time=@n2) {centerX=0, centerY=0, speed=6.2832}) {r=true}
  n10 = ComponentMask(None=Rotator(Coordinate=@n8, Time=Multiply(A=@n2) {constB=-1}) {centerX=0, centerY=0, speed=6.2832}) {g=true}
  n11 = Max(A=@n9, B=@n10)
  n12 = Max(A=AppendVector(A=@n9, B=@n10)) {constB=0}
  n13 = Subtract(A=Add(A=Min(A=@n11) {constB=0}, B=Length(None=@n12) {is=class MaterialExpressionLength}), B=@n7) {constB=0}
  FunctionOutput(None=MaterialFunctionCall(DistanceField=@n13) {materialFunction=&MF_UI_AntiAliasing}.Inner) {outputName=Fill}
  n14 = FunctionInput() {bUsePreviewValueAsDefault=true, inputName=Outline Thickness, inputType=FunctionInput_Scalar, sortPriority=5, previewValue=<0.1 0 0 1>}
  n15 = Subtract(A=Abs(None=FunctionOutput(None=@n13) {outputName=SDF, sortPriority=1}), B=@n14) {constB=0.1}
  FunctionOutput(None=MaterialFunctionCall(DistanceField=FunctionOutput(None=@n15) {outputName=Outline SDF, sortPriority=2}) {materialFunction=&MF_UI_AntiAliasing}.Inner) {outputName=Outline, sortPriority=3}
 ]
}
