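# UI library material function that generates a procedural box with optional per-corner rounding, outputting Fill, Outline, SDF and SDFOutline.
# MF_SDF_Box builds the distance field from UVs, size, roundness, pixel size and an aspect-ratio switch.
# MF_UI_ShapeOutline derives the stroke field from thickness and alignment.
# MF_UI_SDFGlowOrAA converts both fields into Fill and Outline using GlowMin/GlowMax, leaving the raw SDF outputs unaffected by glow.
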MaterialFunction /FortUILibrary/MaterialFunctions/Shapes/MF_UI_Box
graph {
 properties {bExposeToLibrary=true, description=Generates a procedural box shape., previewMaterialDomain=MD_UI, libraryCategoriesText [UI_Library]}
 nodes [
  n1 = FunctionInput(Preview=TextureCoordinate()) {
   bUsePreviewValueAsDefault=true
   description=Controls the UVs of the box with the TextureCoordinate node plugged in. Can be used to manipulate the X/Y coordinates as well.
   inputName=UVs
   inputType=FunctionInput_Vector2
  }
  n2 = FunctionInput() {
   bUsePreviewValueAsDefault=true
   description="The size of the box. Use a Float (Scalar) for an equilateral square or a Float2 for a rectangle with separate Width/Height values."
   inputName=Size
   inputType=FunctionInput_Vector2
   sortPriority=1
   previewValue=<0.5 0.5 0 1>
  }
  n3 = FunctionInput() {
   bUsePreviewValueAsDefault=true
   description="The radius (usually between '0.0' and '0.5') of the rounded corners. Use a Float (Scalar) for uniform roundness, or a Float4 to set a different radius for each corner."
   inputName=CornerRoundness
   inputType=FunctionInput_Vector4
   sortPriority=2
   previewValue=<0 0 0 0>
  }
  n4 = FunctionInput(Preview=StaticBool()) {
   bUsePreviewValueAsDefault=true
   description=Enable to maintain the aspect ratio of rounded corners and glows as the width/height of the box is scaled in either direction.
   inputName=KeepAspectRatio
   inputType=FunctionInput_StaticBool
   sortPriority=8
  }
  n5 = FunctionInput(Preview=MaterialFunctionCall() {materialFunction=&GetUserInterfaceUV}.Pixel Size) {
   bUsePreviewValueAsDefault=true
   description=Pixel Size Reference
   inputName=PixelSize
   inputType=FunctionInput_Vector2
   sortPriority=100
   previewValue=<0.5 0.5 0 1>
  }
  n6 = MaterialFunctionCall!(UVs=@n1, Size=@n2, CornerRoundness=@n3, KeepAspectRatio=@n4, PixelSize=@n5) {materialFunction=/FortUILibrary/MaterialFunctions/Shapes/SDFs/MF_SDF_Box}
  FunctionOutput(None=@n6.SDF) {description=The signed distance field of the shape for further operations., outputName=SDF, sortPriority=2}
  n7 = FunctionInput() {
   bUsePreviewValueAsDefault=true
   description=Determines the thickness of the stroke when using one of the outline outputs.
   inputName=OutlineThickness
   inputType=FunctionInput_Scalar
   sortPriority=3
   previewValue=<0.05 0 0 1>
  }
  n8 = FunctionInput() {
   bUsePreviewValueAsDefault=true
   description="The alignment of the stroke when using the Outline and SDFOutline outputs (0.0 = center, -1.0 = inner, 1.0 = outer)."
   inputName=OutlineAlignment
   inputType=FunctionInput_Scalar
   sortPriority=4
  }
  n9 = MaterialFunctionCall!(DistanceField=@n6.SDF, Threshold, Thickness=@n7, Alignment=@n8, Smoothing) {materialFunction=/FortUILibrary/MaterialFunctions/Shapes/Modifiers/MF_UI_ShapeOutline}
  FunctionOutput(None=@n9.SDFOutline) {description=The signed distance field of the outline for further operations., outputName=SDFOutline, sortPriority=3}
  n10 = FunctionInput() {
   bUsePreviewValueAsDefault=true
   description=Determines glow falloff strength. Can not go above GlowMax value or results will invert. Does not affect SDF or SDFOutline outputs.
   inputName=GlowMin
   inputType=FunctionInput_Scalar
   sortPriority=6
  }
  n11 = FunctionInput() {
   bUsePreviewValueAsDefault=true
   description=Determines glow falloff strength. Can not go below GlowMin value or results will invert. Does not affect SDF or SDFOutline outputs.
   inputName=GlowMax
   inputType=FunctionInput_Scalar
   sortPriority=7
  }
  FunctionOutput(None=MaterialFunctionCall!(DistanceField=@n6.SDF, GlowMin=@n10, GlowMax=@n11) {materialFunction=/FortUILibrary/MaterialFunctions/Shapes/Modifiers/MF_UI_SDFGlowOrAA}.Result) {description=The filled box shape., outputName=Fill}
  FunctionOutput(None=MaterialFunctionCall(DistanceField=@n9.SDFOutline, GlowMin=@n10, GlowMax=@n11) {materialFunction=&MF_UI_SDFGlowOrAA}.Result) {description=The outlined box shape., outputName=Outline, sortPriority=1}
 ]
}
