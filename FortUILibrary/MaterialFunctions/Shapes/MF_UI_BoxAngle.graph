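# UI library material function that masks an angled straight edge from a rectangular UV region, starting at a selectable corner.
# The floored corner index rotates the pivot in quarter turns and combines with the angle in degrees to form a sine/cosine edge normal.
# A dot product gives signed distance from that edge, with Flip reversing the masked side.
# Squared falloff divided by UV screen derivatives sets pixel-relative edge softness before saturation.
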MaterialFunction /FortUILibrary/MaterialFunctions/Shapes/MF_UI_BoxAngle
graph {
 properties {
  bExposeToLibrary=true
  description=Mask an angled edge from a rectangular region
  previewBlendMode=BLEND_Masked
  libraryCategoriesText [UI_Library]
 }
 nodes [
  n1 = FunctionInput() {
   bUsePreviewValueAsDefault=true
   description="0 = top left, 1 = top right, 2 = bottom right, 3 = bottom left"
   inputName=Corner
   inputType=FunctionInput_Scalar
   sortPriority=1
  }
  n2 = Floor(None=@n1)
  n3 = Rotator(Coordinate=Constant2Vector(), Time=@n2) {speed=1.5708}
  n4 = FunctionInput(Preview=TextureCoordinate()) {bUsePreviewValueAsDefault=true, inputName=UV, inputType=FunctionInput_Vector2}
  n5 = FunctionInput(Preview=Constant() {r=15}) {
   bUsePreviewValueAsDefault=true
   description=Value in degrees: 0-90
   inputName=Angle
   inputType=FunctionInput_Scalar
   sortPriority=2
   previewValue=<301111 0 0 1>
  }
  n6 = Add(A=Multiply(A=@n2) {constB=-90}, B=Multiply(A=@n5) {constB=-1})
  n7 = Sine(None=@n6) {period=360}
  n8 = Cosine(None=@n6) {period=360}
  n9 = FunctionInput() {
   bUsePreviewValueAsDefault=true
   description="0 = normal; 1 = flip culling direction"
   inputName=Flip
   inputType=FunctionInput_Scalar
   sortPriority=3
  }
  n10 = Multiply(A=DotProduct(A=Subtract(A=@n4, B=@n3), B=AppendVector(A=@n7, B=@n8)), B=OneMinus(None=Multiply(A=@n9) {constB=2}))
  n11 = FunctionInput() {
   bUsePreviewValueAsDefault=true
   description="~0 soft edge, >1 hard edge"
   inputName=Falloff
   inputType=FunctionInput_Scalar
   sortPriority=4
   previewValue=<1 0 0 1>
  }
  n12 = MaterialFunctionCall(Float2=Abs(None=DDX(Value=@n4))) {materialFunction=&BreakOutFloat2Components}
  n13 = MaterialFunctionCall(Float2=Abs(None=DDY(Value=@n4))) {materialFunction=&BreakOutFloat2Components}
  n14 = Divide(A=Multiply(A=@n11, B=@n11), B=Max(A=Max(A=@n12.R, B=@n13.R), B=Max(A=@n12.G, B=@n13.G)))
  FunctionOutput(None=Saturate(None=Multiply(A=@n10, B=@n14) {constB=100}))
 ]
}
