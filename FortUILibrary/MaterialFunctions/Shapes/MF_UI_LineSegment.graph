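# UI Library material function that draws a procedural line segment between two points, despite its description calling it a circle.
# MF_SDF_LineSegment computes the distance field from UVs and endpoints, then Thickness is subtracted to form the SDF.
# MF_UI_ShapeOutline derives an outline SDF with adjustable thickness and alignment.
# MF_UI_SDFGlowOrAA turns both fields into Fill and Outline using GlowMin/GlowMax, while raw SDF and SDFOutline outputs allow further operations.
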MaterialFunction /FortUILibrary/MaterialFunctions/Shapes/MF_UI_LineSegment
graph {
 properties {bExposeToLibrary=true, description=Generates a procedural circle shape., previewMaterialDomain=MD_UI, libraryCategoriesText [UI_Library]}
 nodes [
  n1 = FunctionInput() {
   bUsePreviewValueAsDefault=true
   description=Determines glow falloff strength. Can not go above GlowMax value or results will invert. Does not affect SDF or SDFOutline outputs.
   inputName=GlowMin
   inputType=FunctionInput_Scalar
   sortPriority=6
  }
  n2 = FunctionInput() {
   bUsePreviewValueAsDefault=true
   description=Determines glow falloff strength. Can not go below GlowMin value or results will invert. Does not affect SDF or SDFOutline outputs.
   inputName=GlowMax
   inputType=FunctionInput_Scalar
   sortPriority=7
  }
  n3 = FunctionInput(Preview=TextureCoordinate()) {
   bUsePreviewValueAsDefault=true
   description=Controls the UVs of the line segment with the TextureCoordinate node plugged in. Can be used to manipulate the X/Y coordinates as well.
   inputName=UVs
   inputType=FunctionInput_Vector2
  }
  n4 = FunctionInput() {
   bUsePreviewValueAsDefault=true
   description=The coordinates of the first point.
   inputName=PointA
   inputType=FunctionInput_Vector2
   sortPriority=1
   previewValue=<-0.5 -0.5 0 1>
  }
  n5 = FunctionInput() {
   bUsePreviewValueAsDefault=true
   description=The coordinates of the second point.
   inputName=PointB
   inputType=FunctionInput_Vector2
   sortPriority=2
   previewValue=<0.5 0.5 0 1>
  }
  n6 = FunctionInput() {
   bUsePreviewValueAsDefault=true
   description=The thickness of the line segment.
   inputName=Thickness
   inputType=FunctionInput_Scalar
   sortPriority=3
   previewValue=<0.1 0 0 1>
  }
  n7 = Subtract(
   A=MaterialFunctionCall!(UVs=@n3, PointA=@n4, PointB=@n5) {materialFunction=/FortUILibrary/MaterialFunctions/Shapes/SDFs/MF_SDF_LineSegment}.SDF
   B=@n6
  )
  FunctionOutput(None=MaterialFunctionCall!(DistanceField=@n7, GlowMin=@n1, GlowMax=@n2) {materialFunction=/FortUILibrary/MaterialFunctions/Shapes/Modifiers/MF_UI_SDFGlowOrAA}.Result) {description=The filled line segment., outputName=Fill}
  FunctionOutput(None=@n7) {description=The signed distance field of the shape for further operations., outputName=SDF, sortPriority=2}
  n8 = FunctionInput() {
   bUsePreviewValueAsDefault=true
   description=Determines the thickness of the stroke when using one of the outline outputs.
   inputName=OutlineThickness
   inputType=FunctionInput_Scalar
   sortPriority=4
   previewValue=<0.05 0 0 1>
  }
  n9 = FunctionInput() {
   bUsePreviewValueAsDefault=true
   description="The alignment of the stroke when using the Outline and SDFOutline outputs (0.0 = center, -1.0 = inner, 1.0 = outer)."
   inputName=OutlineAlignment
   inputType=FunctionInput_Scalar
   sortPriority=5
  }
  n10 = MaterialFunctionCall!(DistanceField=@n7, Threshold, Thickness=@n8, Alignment=@n9, Smoothing) {materialFunction=/FortUILibrary/MaterialFunctions/Shapes/Modifiers/MF_UI_ShapeOutline}
  FunctionOutput(None=MaterialFunctionCall(DistanceField=@n10.SDFOutline, GlowMin=@n1, GlowMax=@n2) {materialFunction=&MF_UI_SDFGlowOrAA}.Result) {description=The outlined line segment., outputName=Outline, sortPriority=1}
  FunctionOutput(None=@n10.SDFOutline) {description=The signed distance field of the outline for further operations., outputName=SDFOutline, sortPriority=3}
 ]
}
