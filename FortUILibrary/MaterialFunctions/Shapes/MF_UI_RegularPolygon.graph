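# UI material function that draws a filled or outlined regular polygon, with optional inner angles for star shapes.
# MF_SDF_RegularPolygon builds a signed distance field from UVs, size, floored side count and inner angle.
# MF_UI_ShapeOutline derives an outline field using thickness and alignment.
# Both fields pass through MF_UI_SDFGlowOrAA with GlowMin/GlowMax for Fill and Outline, while the raw SDF and SDFOutline are also output for further operations.
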MaterialFunction /FortUILibrary/MaterialFunctions/Shapes/MF_UI_RegularPolygon
graph {
 properties {
  bExposeToLibrary=true
  description=Generates a filled or outlined regular polygon shape with the specified number of sides.
  previewMaterialDomain=MD_UI
  libraryCategoriesText [UI_Library]
 }
 nodes [
  n1 = FunctionInput() {
   bUsePreviewValueAsDefault=true
   description=Determines glow falloff strength. Can not go above GlowMax value or results will invert. Does not affect SDF or SDFOutline outputs.
   inputName=GlowMin
   inputType=FunctionInput_Scalar
   sortPriority=6
  }
  n2 = FunctionInput() {
   bUsePreviewValueAsDefault=true
   description=Determines glow falloff strength. Can not go below GlowMin value or results will invert. Does not affect SDF or SDFOutline outputs.
   inputName=GlowMax
   inputType=FunctionInput_Scalar
   sortPriority=7
  }
  n3 = FunctionInput(Preview=TextureCoordinate()) {
   bUsePreviewValueAsDefault=true
   description=Controls the UVs of the polygon with the TextureCoordinate node plugged in. Can be used to manipulate the X/Y coordinates as well.
   inputName=UVs
   inputType=FunctionInput_Vector2
  }
  n4 = FunctionInput() {
   bUsePreviewValueAsDefault=true
   description=The size of the polygon. A '1.0' places the top vertex of the polygon to the border with no glow or outline added.
   inputName=Size
   inputType=FunctionInput_Scalar
   sortPriority=1
   previewValue=<0.5 0 0 1>
  }
  n5 = FunctionInput() {
   bUsePreviewValueAsDefault=true
   description=The number of sides/points on the polygon.
   inputName=Sides
   inputType=FunctionInput_Scalar
   sortPriority=2
   previewValue=<5 0 0 1>
  }
  n6 = FunctionInput() {
   bUsePreviewValueAsDefault=true
   description="Add inner angles to the polygon (eg. to make a star shape)."
   inputName=InnerAngle
   inputType=FunctionInput_Scalar
   sortPriority=3
  }
  n7 = MaterialFunctionCall!(UVs=@n3, Radius=@n4, Sides=Floor(None=@n5), InnerAngle=@n6) {materialFunction=/FortUILibrary/MaterialFunctions/Shapes/SDFs/MF_SDF_RegularPolygon}
  FunctionOutput(None=MaterialFunctionCall!(DistanceField=@n7.SDF, GlowMin=@n1, GlowMax=@n2) {materialFunction=/FortUILibrary/MaterialFunctions/Shapes/Modifiers/MF_UI_SDFGlowOrAA}.Result) {description=The filled polygon shape., outputName=Fill}
  FunctionOutput(None=@n7.SDF) {description=The signed distance field of the shape for further operations., outputName=SDF, sortPriority=2}
  n8 = FunctionInput() {
   bUsePreviewValueAsDefault=true
   description=Determines the thickness of the stroke when using one of the outline outputs.
   inputName=OutlineThickness
   inputType=FunctionInput_Scalar
   sortPriority=4
   previewValue=<0.05 0 0 1>
  }
  n9 = FunctionInput() {
   bUsePreviewValueAsDefault=true
   description="The alignment of the stroke when using the Outline and SDFOutline outputs (0.0 = center, -1.0 = inner, 1.0 = outer)."
   inputName=OutlineAlignment
   inputType=FunctionInput_Scalar
   sortPriority=5
  }
  n10 = MaterialFunctionCall!(DistanceField=@n7.SDF, Threshold, Thickness=@n8, Alignment=@n9, Smoothing) {materialFunction=/FortUILibrary/MaterialFunctions/Shapes/Modifiers/MF_UI_ShapeOutline}
  FunctionOutput(None=MaterialFunctionCall(DistanceField=@n10.SDFOutline, GlowMin=@n1, GlowMax=@n2) {materialFunction=&MF_UI_SDFGlowOrAA}.Result) {description=The outlined polygon shape., outputName=Outline, sortPriority=1}
  FunctionOutput(None=@n10.SDFOutline) {description=The signed distance field of the outline for further operations., outputName=SDFOutline, sortPriority=3}
 ]
}
